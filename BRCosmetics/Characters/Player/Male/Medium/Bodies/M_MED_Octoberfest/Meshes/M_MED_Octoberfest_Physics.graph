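# Physics asset for the M_MED_Octoberfest medium male player body cosmetic, stored alongside its meshes under BRCosmetics.
# It lists no constraints between bodies and shows no body setups, so it defines no visible jointed ragdoll or secondary-motion setup.
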PhysicsAsset /BRCosmetics/Characters/Player/Male/Medium/Bodies/M_MED_Octoberfest/Meshes/M_MED_Octoberfest_Physics
constraints []
