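# Post-pose secondary motion AnimBP for the male medium Octoberfest cosmetic body, copying the attached parent mesh's pose.
# One serial AnimDynamics pass simulates sprung shorts chains (Shorts_Alpha), coat bones with spine/pelvis spherical limits, neck ties and mug.
# Between shorts and coat, Fort Flap nodes remap calf rotation to rotate and Z-scale rear shorts bones.
# The RigidBody node is unconnected and the EventGraph stubs do nothing.
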AnimBlueprint /BRCosmetics/Characters/Player/Male/Medium/Bodies/M_MED_Octoberfest/Meshes/M_MED_Octoberfest_AnimBP
target_skeleton=&/M_MED_Octoberfest_Skeleton
graphs [
 AnimGraph {nodes [
   RigidBody(Alpha=0.85) {
    bOverrideWorldGravity=true
    overridePhysicsAsset=&/M_MED_Octoberfest_Physics
    alphaBoolBlend {blendOption=Linear}
    componentAppliedLinearAccClamp=<1000 1000 1000>
    componentLinearVelScale=<100 100 100>
    overrideWorldGravity {z=-980}
   }
   n1 = Get Shorts_Alpha() -> (Shorts_Alpha:double)
   n2 = "Anim Dynamics - Chain: dyn_shorts_in_l -> dyn_shorts_in_02_l"(ComponentPose=Local To Component(LocalPose=Copy Pose From Mesh() {bUseAttachedParent=true}), Alpha=@n1, ExternalForce="0,0,0") {
    angularDampingOverride=0.999
    angularSpringConstant=100
    bAngularSpring=true
    bChain=true
    bGravityOverrideInSimSpace=true
    bOverrideAngularDamping=true
    numSolverIterationsPostUpdate=4
    numSolverIterationsPreUpdate=16
    alphaBoolBlend {blendOption=Linear}
    boundBone {boneName=dyn_shorts_in_l}
    chainEnd {boneName=dyn_shorts_in_02_l}
    physicsBodyDefinitions [{
      sphereCollisionRadius=0
      boundBone {boneName=dyn_shorts_in_l}
      boxExtents=<25 30 12>
      constraintSetup {angularTargetAxis=AxisZ, twistAxis=AxisZ, angularLimitsMax=<10 10 5>, angularLimitsMin=<-10 -10 -5>, angularTarget=<0 0 1>}
      localJointOffset=<0 0 -3.5>
     }]
   }
   n3 = "Anim Dynamics - Chain: dyn_shorts_out_l -> dyn_shorts_out_02_l"(ComponentPose=@n2, Alpha=@n1, ExternalForce="0,0,0") {
    angularDampingOverride=0.999
    angularSpringConstant=100
    bAngularSpring=true
    bChain=true
    bGravityOverrideInSimSpace=true
    bOverrideAngularDamping=true
    numSolverIterationsPostUpdate=4
    numSolverIterationsPreUpdate=16
    alphaBoolBlend {blendOption=Linear}
    boundBone {boneName=dyn_shorts_out_l}
    chainEnd {boneName=dyn_shorts_out_02_l}
    physicsBodyDefinitions [{
      sphereCollisionRadius=0
      boundBone {boneName=dyn_shorts_out_l}
      boxExtents=<50 50 30>
      constraintSetup {angularTargetAxis=AxisZ, twistAxis=AxisZ, angularLimitsMax=<10 10 5>, angularLimitsMin=<-10 -10 -5>, angularTarget=<0 0 1>}
      localJointOffset=<0 0 -6.5>
     }]
   }
   n4 = "Anim Dynamics - Chain: dyn_shorts_fr_l -> dyn_shorts_fr_02_l"(ComponentPose=@n3, Alpha=@n1, ExternalForce="0,0,0") {
    angularDampingOverride=0.999
    angularSpringConstant=300
    bAngularSpring=true
    bChain=true
    bGravityOverrideInSimSpace=true
    bOverrideAngularDamping=true
    numSolverIterationsPostUpdate=4
    numSolverIterationsPreUpdate=16
    alphaBoolBlend {blendOption=Linear}
    boundBone {boneName=dyn_shorts_fr_l}
    chainEnd {boneName=dyn_shorts_fr_02_l}
    physicsBodyDefinitions [{
      sphereCollisionRadius=0
      boundBone {boneName=dyn_shorts_fr_l}
      boxExtents=<18 25 8>
      constraintSetup {angularTargetAxis=AxisZ, twistAxis=AxisZ, angularLimitsMax=<10 0 5>, angularLimitsMin=<-10 -20 -5>, angularTarget=<0 0 1>}
      localJointOffset=<0 0 -5>
     }]
   }
   n5 = "Anim Dynamics - Chain: dyn_shorts_bk_l -> dyn_shorts_bk_02_l"(ComponentPose=@n4, Alpha=@n1, ExternalForce="0,0,0") {
    angularDampingOverride=0.999
    angularSpringConstant=300
    bAngularSpring=true
    bChain=true
    bGravityOverrideInSimSpace=true
    bOverrideAngularDamping=true
    numSolverIterationsPostUpdate=4
    numSolverIterationsPreUpdate=16
    alphaBoolBlend {blendOption=Linear}
    boundBone {boneName=dyn_shorts_bk_l}
    chainEnd {boneName=dyn_shorts_bk_02_l}
    physicsBodyDefinitions [{
      sphereCollisionRadius=0
      boundBone {boneName=dyn_shorts_bk_l}
      boxExtents=<40 38 15>
      constraintSetup {angularTargetAxis=AxisZ, twistAxis=AxisZ, angularLimitsMax=<10 0 5>, angularLimitsMin=<-10 -20 -5>, angularTarget=<0 0 1>}
      localJointOffset=<0 0 -3.5>
     }]
   }
   n6 = "Anim Dynamics - Chain: dyn_shorts_in_r -> dyn_shorts_in_02_r"(ComponentPose=@n5, Alpha=@n1, ExternalForce="0,0,0") {
    angularDampingOverride=0.999
    angularSpringConstant=100
    bAngularSpring=true
    bChain=true
    bGravityOverrideInSimSpace=true
    bOverrideAngularDamping=true
    numSolverIterationsPostUpdate=4
    numSolverIterationsPreUpdate=16
    alphaBoolBlend {blendOption=Linear}
    boundBone {boneName=dyn_shorts_in_r}
    chainEnd {boneName=dyn_shorts_in_02_r}
    physicsBodyDefinitions [{
      sphereCollisionRadius=0
      boundBone {boneName=dyn_shorts_in_r}
      boxExtents=<25 29 12>
      constraintSetup {angularTargetAxis=AxisZ, twistAxis=AxisZ, angularLimitsMax=<10 10 5>, angularLimitsMin=<-10 -10 -5>, angularTarget=<0 0 1>}
      localJointOffset=<0 0 3.5>
     }]
   }
   n7 = "Anim Dynamics - Chain: dyn_shorts_out_r -> dyn_shorts_out_02_r"(ComponentPose=@n6, Alpha=@n1, ExternalForce="0,0,0") {
    angularDampingOverride=0.999
    angularSpringConstant=100
    bAngularSpring=true
    bChain=true
    bGravityOverrideInSimSpace=true
    bOverrideAngularDamping=true
    numSolverIterationsPostUpdate=4
    numSolverIterationsPreUpdate=16
    alphaBoolBlend {blendOption=Linear}
    boundBone {boneName=dyn_shorts_out_r}
    chainEnd {boneName=dyn_shorts_out_02_r}
    physicsBodyDefinitions [{
      sphereCollisionRadius=0
      boundBone {boneName=dyn_shorts_out_r}
      boxExtents=<50 50 30>
      constraintSetup {angularTargetAxis=AxisZ, twistAxis=AxisZ, angularLimitsMax=<10 10 5>, angularLimitsMin=<-10 -10 -5>, angularTarget=<0 0 1>}
      localJointOffset=<0 0 6.5>
     }]
   }
   n8 = "Anim Dynamics - Chain: dyn_shorts_fr_r -> dyn_shorts_fr_02_r"(ComponentPose=@n7, Alpha=@n1, ExternalForce="0,0,0") {
    angularDampingOverride=0.999
    angularSpringConstant=300
    bAngularSpring=true
    bChain=true
    bGravityOverrideInSimSpace=true
    bOverrideAngularDamping=true
    numSolverIterationsPostUpdate=4
    numSolverIterationsPreUpdate=16
    alphaBoolBlend {blendOption=Linear}
    boundBone {boneName=dyn_shorts_fr_r}
    chainEnd {boneName=dyn_shorts_fr_02_r}
    physicsBodyDefinitions [{
      sphereCollisionRadius=0
      boundBone {boneName=dyn_shorts_fr_r}
      boxExtents=<26 22 8>
      constraintSetup {angularTargetAxis=AxisZ, twistAxis=AxisZ, angularLimitsMax=<10 0 5>, angularLimitsMin=<-10 -20 -5>, angularTarget=<0 0 1>}
      localJointOffset=<0 0 5>
     }]
   }
   n9 = "Anim Dynamics - Chain: dyn_shorts_bk_r -> dyn_shorts_bk_02_r"(ComponentPose=@n8, Alpha=@n1, ExternalForce="0,0,0") {
    angularDampingOverride=0.999
    angularSpringConstant=300
    bAngularSpring=true
    bChain=true
    bGravityOverrideInSimSpace=true
    bOverrideAngularDamping=true
    numSolverIterationsPostUpdate=4
    numSolverIterationsPreUpdate=16
    alphaBoolBlend {blendOption=Linear}
    boundBone {boneName=dyn_shorts_bk_r}
    chainEnd {boneName=dyn_shorts_bk_02_r}
    physicsBodyDefinitions [{
      sphereCollisionRadius=0
      boundBone {boneName=dyn_shorts_bk_r}
      boxExtents=<40 36 15>
      constraintSetup {angularTargetAxis=AxisZ, twistAxis=AxisZ, angularLimitsMax=<0 0 5>, angularLimitsMin=<0 -20 -5>, angularTarget=<0 0 1>}
      localJointOffset=<0 0 3.5>
     }]
   }
   n10 = Flap Node - Bone: dyn_shorts_bk_02_l!(ComponentPose:Component Space Pose Link Structure=@n9, Alpha:float=1) -> (Pose:Component Space Pose Link Structure) {
    is=title FortAnimGraphNode_Flap
    targetComponent=RotationY
    alphaBoolBlend {blendOption=Linear}
    sourceDrives [{
      bUseRange=true
      drivingCurve=null
      multiplier=1
      rangeMax=-0.53
      rangeMin=-2.094
      remappedMax=0
      remappedMin=-1.919
      sourceComponent=RotationZ
      useQuaternion=true
      sourceBone {boneName=calf_l}
     }]
    targetBone {boneName=dyn_shorts_bk_02_l}
   }
   n11 = Flap Node - Bone: dyn_shorts_bk_02_l(ComponentPose=@n10, Alpha=1) {
    targetComponent=ScaleZ
    alphaBoolBlend {blendOption=Linear}
    sourceDrives [{
      bUseRange=true
      drivingCurve=null
      multiplier=1
      rangeMax=-0.53
      rangeMin=-2.094
      remappedMax=0
      remappedMin=-0.7
      sourceComponent=RotationZ
      useQuaternion=true
      sourceBone {boneName=calf_l}
     }]
    targetBone {boneName=dyn_shorts_bk_02_l}
   }
   n12 = Flap Node - Bone: dyn_shorts_bk_02_r!(ComponentPose:Component Space Pose Link Structure=@n11, Alpha:float=1) -> (Pose:Component Space Pose Link Structure) {
    is=title FortAnimGraphNode_Flap
    targetComponent=RotationY
    alphaBoolBlend {blendOption=Linear}
    sourceDrives [{
      bUseRange=true
      drivingCurve=null
      multiplier=1
      rangeMax=-0.53
      rangeMin=-2.094
      remappedMax=0
      remappedMin=-1.99
      sourceComponent=RotationZ
      useQuaternion=true
      sourceBone {boneName=calf_r}
     }]
    targetBone {boneName=dyn_shorts_bk_02_r}
   }
   n13 = Flap Node - Bone: dyn_shorts_bk_02_r(ComponentPose=@n12, Alpha=1) {
    targetComponent=ScaleZ
    alphaBoolBlend {blendOption=Linear}
    sourceDrives [{
      bUseRange=true
      drivingCurve=null
      multiplier=1
      rangeMax=-0.53
      rangeMin=-2.094
      remappedMax=0
      remappedMin=-0.75
      sourceComponent=RotationZ
      useQuaternion=true
      sourceBone {boneName=calf_r}
     }]
    targetBone {boneName=dyn_shorts_bk_02_r}
   }
   n14 = Anim Dynamics - Bone: dyn_coat_bk_01(ComponentPose=@n13, Alpha=1, ExternalForce="0,0,0") {
    angularDampingOverride=0.98
    bGravityOverrideInSimSpace=true
    bOverrideAngularDamping=true
    bUsePlanarLimit=false
    bUseSphericalLimits=true
    numSolverIterationsPostUpdate=4
    numSolverIterationsPreUpdate=16
    alphaBoolBlend {blendOption=Linear}
    boundBone {boneName=dyn_coat_bk_01}
    componentAppliedLinearAccClamp=<3000 3000 3000>
    componentLinearAccScale=<0.5 0.5 0.5>
    componentLinearVelScale=<0.5 0.5 0.5>
    physicsBodyDefinitions [{
      sphereCollisionRadius=0
      boundBone {boneName=dyn_coat_bk_01}
      boxExtents=<15 30 40>
      constraintSetup {twistAxis=AxisZ, angularLimitsMax=<89 30 0>, angularLimitsMin=<-89 -150 0>, angularTarget=<0 0 0>, linearAxesMin=<-15 0 0>}
      localJointOffset=<0 0 10>
     }]
    sphericalLimits [
     {limitRadius=18.5, drivingBone {boneName=spine_01}, sphereLocalOffset=<0 0 0>}
     {limitRadius=18.5, drivingBone {boneName=spine_02}, sphereLocalOffset=<0 0 0>}
     {limitRadius=18.25, drivingBone {boneName=pelvis}, sphereLocalOffset=<0 0 0>}
    ]
   }
   n15 = Anim Dynamics - Bone: dyn_coat_bk_02(ComponentPose=@n14, Alpha=1, ExternalForce="0,0,0") {
    angularDampingOverride=0.98
    bGravityOverrideInSimSpace=true
    bOverrideAngularDamping=true
    bUsePlanarLimit=false
    bUseSphericalLimits=true
    numSolverIterationsPostUpdate=4
    numSolverIterationsPreUpdate=16
    alphaBoolBlend {blendOption=Linear}
    boundBone {boneName=dyn_coat_bk_02}
    componentAppliedLinearAccClamp=<3000 3000 3000>
    componentLinearAccScale=<0.5 0.5 0.5>
    componentLinearVelScale=<0.5 0.5 0.5>
    physicsBodyDefinitions [{
      sphereCollisionRadius=0
      boundBone {boneName=dyn_coat_bk_02}
      boxExtents=<15 30 40>
      constraintSetup {twistAxis=AxisZ, angularLimitsMax=<89 30 0>, angularLimitsMin=<-89 -150 0>, angularTarget=<0 0 0>, linearAxesMin=<-15 0 0>}
      localJointOffset=<0 0 10>
     }]
    sphericalLimits [
     {limitRadius=18.5, drivingBone {boneName=spine_01}, sphereLocalOffset=<0 0 0>}
     {limitRadius=18.5, drivingBone {boneName=spine_02}, sphereLocalOffset=<0 0 0>}
     {limitRadius=18.25, drivingBone {boneName=pelvis}, sphereLocalOffset=<0 0 0>}
    ]
   }
   n16 = Anim Dynamics - Bone: dyn_coat_bk_01_r(ComponentPose=@n15, Alpha=1, ExternalForce="0,0,0") {
    angularDampingOverride=0.98
    bGravityOverrideInSimSpace=true
    bOverrideAngularDamping=true
    bUsePlanarLimit=false
    bUseSphericalLimits=true
    numSolverIterationsPostUpdate=4
    numSolverIterationsPreUpdate=16
    alphaBoolBlend {blendOption=Linear}
    boundBone {boneName=dyn_coat_bk_01_r}
    componentAppliedLinearAccClamp=<3000 3000 3000>
    componentLinearAccScale=<0.5 0.5 0.5>
    componentLinearVelScale=<0.5 0.5 0.5>
    physicsBodyDefinitions [{
      sphereCollisionRadius=0
      boundBone {boneName=dyn_coat_bk_01_r}
      boxExtents=<15 30 40>
      constraintSetup {twistAxis=AxisZ, angularLimitsMax=<89 30 0>, angularLimitsMin=<-89 -150 0>, angularTarget=<0 0 0>, linearAxesMin=<-15 0 0>}
      localJointOffset=<0 0 10>
     }]
    sphericalLimits [
     {limitRadius=18.25, drivingBone {boneName=spine_01}, sphereLocalOffset=<0 0 0>}
     {limitRadius=18.25, drivingBone {boneName=spine_02}, sphereLocalOffset=<0 0 0>}
     {limitRadius=18.25, drivingBone {boneName=pelvis}, sphereLocalOffset=<0 0 0>}
    ]
   }
   n17 = Anim Dynamics - Bone: dyn_coat_bk_02_r(ComponentPose=@n16, Alpha=1, ExternalForce="0,0,0") {
    angularDampingOverride=0.98
    bGravityOverrideInSimSpace=true
    bOverrideAngularDamping=true
    bUsePlanarLimit=false
    bUseSphericalLimits=true
    numSolverIterationsPostUpdate=4
    numSolverIterationsPreUpdate=16
    alphaBoolBlend {blendOption=Linear}
    boundBone {boneName=dyn_coat_bk_02_r}
    componentAppliedLinearAccClamp=<3000 3000 3000>
    componentLinearAccScale=<0.5 0.5 0.5>
    componentLinearVelScale=<0.5 0.5 0.5>
    physicsBodyDefinitions [{
      sphereCollisionRadius=0
      boundBone {boneName=dyn_coat_bk_02_r}
      boxExtents=<15 30 40>
      constraintSetup {twistAxis=AxisZ, angularLimitsMax=<89 30 0>, angularLimitsMin=<-89 -150 0>, angularTarget=<0 0 0>, linearAxesMin=<-15 0 0>}
      localJointOffset=<0 0 10>
     }]
    sphericalLimits [
     {limitRadius=18.25, drivingBone {boneName=spine_01}, sphereLocalOffset=<0 0 0>}
     {limitRadius=18.25, drivingBone {boneName=spine_02}, sphereLocalOffset=<0 0 0>}
     {limitRadius=18.25, drivingBone {boneName=pelvis}, sphereLocalOffset=<0 0 0>}
    ]
   }
   n18 = Anim Dynamics - Bone: dyn_coat_bk_01_l(ComponentPose=@n17, Alpha=1, ExternalForce="0,0,0") {
    angularDampingOverride=0.98
    bGravityOverrideInSimSpace=true
    bOverrideAngularDamping=true
    bUsePlanarLimit=false
    bUseSphericalLimits=true
    numSolverIterationsPostUpdate=4
    numSolverIterationsPreUpdate=16
    alphaBoolBlend {blendOption=Linear}
    boundBone {boneName=dyn_coat_bk_01_l}
    componentAppliedLinearAccClamp=<3000 3000 3000>
    componentLinearAccScale=<0.5 0.5 0.5>
    componentLinearVelScale=<0.5 0.5 0.5>
    physicsBodyDefinitions [{
      sphereCollisionRadius=0
      boundBone {boneName=dyn_coat_bk_01_l}
      boxExtents=<15 30 40>
      constraintSetup {twistAxis=AxisZ, angularLimitsMax=<89 30 0>, angularLimitsMin=<-89 -150 0>, angularTarget=<0 0 0>, linearAxesMax=<15 0 0>}
      localJointOffset=<0 0 -10>
     }]
    sphericalLimits [
     {limitRadius=18.25, drivingBone {boneName=spine_01}, sphereLocalOffset=<0 0 0>}
     {limitRadius=18.25, drivingBone {boneName=spine_02}, sphereLocalOffset=<0 0 0>}
     {limitRadius=18.25, drivingBone {boneName=pelvis}, sphereLocalOffset=<0 0 0>}
    ]
   }
   n19 = Anim Dynamics - Bone: dyn_coat_bk_02_l(ComponentPose=@n18, Alpha=1, ExternalForce="0,0,0") {
    angularDampingOverride=0.98
    bGravityOverrideInSimSpace=true
    bOverrideAngularDamping=true
    bUsePlanarLimit=false
    bUseSphericalLimits=true
    numSolverIterationsPostUpdate=4
    numSolverIterationsPreUpdate=16
    alphaBoolBlend {blendOption=Linear}
    boundBone {boneName=dyn_coat_bk_02_l}
    componentAppliedLinearAccClamp=<3000 3000 3000>
    componentLinearAccScale=<0.5 0.5 0.5>
    componentLinearVelScale=<0.5 0.5 0.5>
    physicsBodyDefinitions [{
      sphereCollisionRadius=0
      boundBone {boneName=dyn_coat_bk_02_l}
      boxExtents=<15 30 40>
      constraintSetup {twistAxis=AxisZ, angularLimitsMax=<89 30 0>, angularLimitsMin=<-89 -150 0>, angularTarget=<0 0 0>, linearAxesMax=<15 0 0>}
      localJointOffset=<0 0 -10>
     }]
    sphericalLimits [
     {limitRadius=18.25, drivingBone {boneName=spine_01}, sphereLocalOffset=<0 0 0>}
     {limitRadius=18.25, drivingBone {boneName=spine_02}, sphereLocalOffset=<0 0 0>}
     {limitRadius=18.25, drivingBone {boneName=pelvis}, sphereLocalOffset=<0 0 0>}
    ]
   }
   n20 = Anim Dynamics - Bone: dyn_coat_out_01_l(ComponentPose=@n19, Alpha=1, ExternalForce="0,0,0") {
    angularDampingOverride=0.98
    bGravityOverrideInSimSpace=true
    bOverrideAngularDamping=true
    bUsePlanarLimit=false
    bUseSphericalLimits=true
    numSolverIterationsPostUpdate=4
    numSolverIterationsPreUpdate=16
    alphaBoolBlend {blendOption=Linear}
    boundBone {boneName=dyn_coat_out_01_l}
    componentAppliedLinearAccClamp=<3000 3000 3000>
    componentLinearAccScale=<0.5 0.5 0.5>
    componentLinearVelScale=<0.5 0.5 0.5>
    physicsBodyDefinitions [{
      sphereCollisionRadius=0
      boundBone {boneName=dyn_coat_out_01_l}
      boxExtents=<15 30 40>
      constraintSetup {twistAxis=AxisZ, angularLimitsMax=<89 30 0>, angularLimitsMin=<-89 -150 0>, angularTarget=<0 0 0>}
      localJointOffset=<0 0 -10>
     }]
    sphericalLimits [
     {limitRadius=18, drivingBone {boneName=spine_01}, sphereLocalOffset=<0 0 0>}
     {limitRadius=18, drivingBone {boneName=spine_02}, sphereLocalOffset=<0 0 0>}
     {limitRadius=18, drivingBone {boneName=pelvis}, sphereLocalOffset=<0 0 0>}
    ]
   }
   n21 = Anim Dynamics - Bone: dyn_coat_out_02_l(ComponentPose=@n20, Alpha=1, ExternalForce="0,0,0") {
    angularDampingOverride=0.98
    bGravityOverrideInSimSpace=true
    bOverrideAngularDamping=true
    bUsePlanarLimit=false
    bUseSphericalLimits=true
    numSolverIterationsPostUpdate=4
    numSolverIterationsPreUpdate=16
    alphaBoolBlend {blendOption=Linear}
    boundBone {boneName=dyn_coat_out_02_l}
    componentAppliedLinearAccClamp=<3000 3000 3000>
    componentLinearAccScale=<0.5 0.5 0.5>
    componentLinearVelScale=<0.5 0.5 0.5>
    physicsBodyDefinitions [{
      sphereCollisionRadius=0
      boundBone {boneName=dyn_coat_out_02_l}
      boxExtents=<15 30 40>
      constraintSetup {twistAxis=AxisZ, angularLimitsMax=<89 30 0>, angularLimitsMin=<-89 -150 0>, angularTarget=<0 0 0>}
      localJointOffset=<0 0 -10>
     }]
    sphericalLimits [
     {limitRadius=18, drivingBone {boneName=spine_01}, sphereLocalOffset=<0 0 0>}
     {limitRadius=18, drivingBone {boneName=spine_02}, sphereLocalOffset=<0 0 0>}
     {limitRadius=18, drivingBone {boneName=pelvis}, sphereLocalOffset=<0 0 0>}
    ]
   }
   n22 = Anim Dynamics - Bone: dyn_coat_fr_01_l(ComponentPose=@n21, Alpha=1, ExternalForce="0,0,0") {
    angularDampingOverride=0.98
    bGravityOverrideInSimSpace=true
    bOverrideAngularDamping=true
    bUsePlanarLimit=false
    bUseSphericalLimits=true
    numSolverIterationsPostUpdate=4
    numSolverIterationsPreUpdate=16
    alphaBoolBlend {blendOption=Linear}
    boundBone {boneName=dyn_coat_fr_01_l}
    componentAppliedLinearAccClamp=<3000 3000 3000>
    componentLinearAccScale=<0.5 0.5 0.5>
    componentLinearVelScale=<0.5 0.5 0.5>
    physicsBodyDefinitions [{
      sphereCollisionRadius=0
      boundBone {boneName=dyn_coat_fr_01_l}
      boxExtents=<15 30 40>
      constraintSetup {twistAxis=AxisZ, angularLimitsMax=<89 30 0>, angularLimitsMin=<-89 -150 0>, angularTarget=<0 0 0>}
      localJointOffset=<0 0 -10>
     }]
    sphericalLimits [
     {limitRadius=18, drivingBone {boneName=spine_01}, sphereLocalOffset=<0 0 0>}
     {limitRadius=18, drivingBone {boneName=spine_02}, sphereLocalOffset=<0 0 0>}
     {limitRadius=18, drivingBone {boneName=pelvis}, sphereLocalOffset=<0 0 0>}
    ]
   }
   n23 = Anim Dynamics - Bone: dyn_coat_fr_02_l(ComponentPose=@n22, Alpha=1, ExternalForce="0,0,0") {
    angularDampingOverride=0.98
    bGravityOverrideInSimSpace=true
    bOverrideAngularDamping=true
    bUsePlanarLimit=false
    bUseSphericalLimits=true
    numSolverIterationsPostUpdate=4
    numSolverIterationsPreUpdate=16
    alphaBoolBlend {blendOption=Linear}
    boundBone {boneName=dyn_coat_fr_02_l}
    componentAppliedLinearAccClamp=<3000 3000 3000>
    componentLinearAccScale=<0.5 0.5 0.5>
    componentLinearVelScale=<0.5 0.5 0.5>
    physicsBodyDefinitions [{
      sphereCollisionRadius=0
      boundBone {boneName=dyn_coat_fr_02_l}
      boxExtents=<15 30 40>
      constraintSetup {twistAxis=AxisZ, angularLimitsMax=<89 30 0>, angularLimitsMin=<-89 -150 0>, angularTarget=<0 0 0>}
      localJointOffset=<0 0 -10>
     }]
    sphericalLimits [
     {limitRadius=18, drivingBone {boneName=spine_01}, sphereLocalOffset=<0 0 0>}
     {limitRadius=18, drivingBone {boneName=spine_02}, sphereLocalOffset=<0 0 0>}
     {limitRadius=18, drivingBone {boneName=pelvis}, sphereLocalOffset=<0 0 0>}
    ]
   }
   n24 = Anim Dynamics - Bone: dyn_coat_fr_01_r(ComponentPose=@n23, Alpha=1, ExternalForce="0,0,0") {
    angularDampingOverride=0.98
    bGravityOverrideInSimSpace=true
    bOverrideAngularDamping=true
    bUsePlanarLimit=false
    bUseSphericalLimits=true
    numSolverIterationsPostUpdate=4
    numSolverIterationsPreUpdate=16
    alphaBoolBlend {blendOption=Linear}
    boundBone {boneName=dyn_coat_fr_01_r}
    componentAppliedLinearAccClamp=<3000 3000 3000>
    componentLinearAccScale=<0.5 0.5 0.5>
    componentLinearVelScale=<0.5 0.5 0.5>
    physicsBodyDefinitions [{
      sphereCollisionRadius=0
      boundBone {boneName=dyn_coat_fr_01_r}
      boxExtents=<15 30 40>
      constraintSetup {twistAxis=AxisZ, angularLimitsMax=<89 30 0>, angularLimitsMin=<-89 -150 0>, angularTarget=<0 0 0>}
      localJointOffset=<0 0 10>
     }]
    sphericalLimits [
     {limitRadius=18, drivingBone {boneName=spine_01}, sphereLocalOffset=<0 0 0>}
     {limitRadius=18, drivingBone {boneName=spine_02}, sphereLocalOffset=<0 0 0>}
     {limitRadius=18, drivingBone {boneName=pelvis}, sphereLocalOffset=<0 0 0>}
    ]
   }
   n25 = Anim Dynamics - Bone: dyn_coat_fr_02_r(ComponentPose=@n24, Alpha=1, ExternalForce="0,0,0") {
    angularDampingOverride=0.98
    bGravityOverrideInSimSpace=true
    bOverrideAngularDamping=true
    bUsePlanarLimit=false
    bUseSphericalLimits=true
    numSolverIterationsPostUpdate=4
    numSolverIterationsPreUpdate=16
    alphaBoolBlend {blendOption=Linear}
    boundBone {boneName=dyn_coat_fr_02_r}
    componentAppliedLinearAccClamp=<3000 3000 3000>
    componentLinearAccScale=<0.5 0.5 0.5>
    componentLinearVelScale=<0.5 0.5 0.5>
    physicsBodyDefinitions [{
      sphereCollisionRadius=0
      boundBone {boneName=dyn_coat_fr_02_r}
      boxExtents=<15 30 40>
      constraintSetup {twistAxis=AxisZ, angularLimitsMax=<89 30 0>, angularLimitsMin=<-89 -150 0>, angularTarget=<0 0 0>}
      localJointOffset=<0 0 10>
     }]
    sphericalLimits [
     {limitRadius=18, drivingBone {boneName=spine_01}, sphereLocalOffset=<0 0 0>}
     {limitRadius=18, drivingBone {boneName=spine_02}, sphereLocalOffset=<0 0 0>}
     {limitRadius=18, drivingBone {boneName=pelvis}, sphereLocalOffset=<0 0 0>}
    ]
   }
   n26 = Anim Dynamics - Bone: dyn_coat_out_01_r(ComponentPose=@n25, Alpha=1, ExternalForce="0,0,0") {
    angularDampingOverride=0.98
    bGravityOverrideInSimSpace=true
    bOverrideAngularDamping=true
    bUsePlanarLimit=false
    bUseSphericalLimits=true
    numSolverIterationsPostUpdate=4
    numSolverIterationsPreUpdate=16
    alphaBoolBlend {blendOption=Linear}
    boundBone {boneName=dyn_coat_out_01_r}
    componentAppliedLinearAccClamp=<3000 3000 3000>
    componentLinearAccScale=<0.5 0.5 0.5>
    componentLinearVelScale=<0.5 0.5 0.5>
    physicsBodyDefinitions [{
      sphereCollisionRadius=0
      boundBone {boneName=dyn_coat_out_01_r}
      boxExtents=<15 30 40>
      constraintSetup {twistAxis=AxisZ, angularLimitsMax=<89 30 0>, angularLimitsMin=<-89 -150 0>, angularTarget=<0 0 0>}
      localJointOffset=<0 0 10>
     }]
    sphericalLimits [
     {limitRadius=18, drivingBone {boneName=spine_01}, sphereLocalOffset=<0 0 0>}
     {limitRadius=18, drivingBone {boneName=spine_02}, sphereLocalOffset=<0 0 0>}
     {limitRadius=18, drivingBone {boneName=pelvis}, sphereLocalOffset=<0 0 0>}
    ]
   }
   n27 = Anim Dynamics - Bone: dyn_coat_out_02_r(ComponentPose=@n26, Alpha=1, ExternalForce="0,0,0") {
    angularDampingOverride=0.98
    bGravityOverrideInSimSpace=true
    bOverrideAngularDamping=true
    bUsePlanarLimit=false
    bUseSphericalLimits=true
    numSolverIterationsPostUpdate=4
    numSolverIterationsPreUpdate=16
    alphaBoolBlend {blendOption=Linear}
    boundBone {boneName=dyn_coat_out_02_r}
    componentAppliedLinearAccClamp=<3000 3000 3000>
    componentLinearAccScale=<0.5 0.5 0.5>
    componentLinearVelScale=<0.5 0.5 0.5>
    physicsBodyDefinitions [{
      sphereCollisionRadius=0
      boundBone {boneName=dyn_coat_out_02_r}
      boxExtents=<15 30 40>
      constraintSetup {twistAxis=AxisZ, angularLimitsMax=<89 30 0>, angularLimitsMin=<-89 -150 0>, angularTarget=<0 0 0>}
      localJointOffset=<0 0 10>
     }]
    sphericalLimits [
     {limitRadius=18, drivingBone {boneName=spine_01}, sphereLocalOffset=<0 0 0>}
     {limitRadius=18, drivingBone {boneName=spine_02}, sphereLocalOffset=<0 0 0>}
     {limitRadius=18, drivingBone {boneName=pelvis}, sphereLocalOffset=<0 0 0>}
    ]
   }
   n28 = Anim Dynamics - Bone: dyn_coat_frdiag_01_l(ComponentPose=@n27, Alpha=1, ExternalForce="0,0,0") {
    angularDampingOverride=0.98
    bGravityOverrideInSimSpace=true
    bOverrideAngularDamping=true
    bUsePlanarLimit=false
    bUseSphericalLimits=true
    numSolverIterationsPostUpdate=4
    numSolverIterationsPreUpdate=16
    alphaBoolBlend {blendOption=Linear}
    boundBone {boneName=dyn_coat_frdiag_01_l}
    componentAppliedLinearAccClamp=<3000 3000 3000>
    componentLinearAccScale=<0.5 0.5 0.5>
    componentLinearVelScale=<0.5 0.5 0.5>
    physicsBodyDefinitions [{
      sphereCollisionRadius=0
      boundBone {boneName=dyn_coat_frdiag_01_l}
      boxExtents=<15 30 40>
      constraintSetup {twistAxis=AxisZ, angularLimitsMax=<89 30 0>, angularLimitsMin=<-89 -150 0>, angularTarget=<0 0 0>}
      localJointOffset=<0 0 -10>
     }]
    sphericalLimits [
     {limitRadius=18, drivingBone {boneName=spine_01}, sphereLocalOffset=<0 0 0>}
     {limitRadius=18, drivingBone {boneName=spine_02}, sphereLocalOffset=<0 0 0>}
     {limitRadius=18, drivingBone {boneName=pelvis}, sphereLocalOffset=<0 0 0>}
    ]
   }
   n29 = Anim Dynamics - Bone: dyn_coat_frdiag_02_l(ComponentPose=@n28, Alpha=1, ExternalForce="0,0,0") {
    angularDampingOverride=0.98
    bGravityOverrideInSimSpace=true
    bOverrideAngularDamping=true
    bUsePlanarLimit=false
    bUseSphericalLimits=true
    numSolverIterationsPostUpdate=4
    numSolverIterationsPreUpdate=16
    alphaBoolBlend {blendOption=Linear}
    boundBone {boneName=dyn_coat_frdiag_02_l}
    componentAppliedLinearAccClamp=<3000 3000 3000>
    componentLinearAccScale=<0.5 0.5 0.5>
    componentLinearVelScale=<0.5 0.5 0.5>
    physicsBodyDefinitions [{
      sphereCollisionRadius=0
      boundBone {boneName=dyn_coat_frdiag_02_l}
      boxExtents=<15 30 40>
      constraintSetup {twistAxis=AxisZ, angularLimitsMax=<89 30 0>, angularLimitsMin=<-89 -150 0>, angularTarget=<0 0 0>}
      localJointOffset=<0 0 -10>
     }]
    sphericalLimits [
     {limitRadius=18, drivingBone {boneName=spine_01}, sphereLocalOffset=<0 0 0>}
     {limitRadius=18, drivingBone {boneName=spine_02}, sphereLocalOffset=<0 0 0>}
     {limitRadius=18, drivingBone {boneName=pelvis}, sphereLocalOffset=<0 0 0>}
    ]
   }
   n30 = Anim Dynamics - Bone: dyn_coat_frdiag_01_r(ComponentPose=@n29, Alpha=1, ExternalForce="0,0,0") {
    angularDampingOverride=0.98
    bGravityOverrideInSimSpace=true
    bOverrideAngularDamping=true
    bUsePlanarLimit=false
    bUseSphericalLimits=true
    numSolverIterationsPostUpdate=4
    numSolverIterationsPreUpdate=16
    alphaBoolBlend {blendOption=Linear}
    boundBone {boneName=dyn_coat_frdiag_01_r}
    componentAppliedLinearAccClamp=<3000 3000 3000>
    componentLinearAccScale=<0.5 0.5 0.5>
    componentLinearVelScale=<0.5 0.5 0.5>
    physicsBodyDefinitions [{
      sphereCollisionRadius=0
      boundBone {boneName=dyn_coat_frdiag_01_r}
      boxExtents=<15 30 40>
      constraintSetup {twistAxis=AxisZ, angularLimitsMax=<89 30 0>, angularLimitsMin=<-89 -150 0>, angularTarget=<0 0 0>}
      localJointOffset=<0 0 10>
     }]
    sphericalLimits [
     {limitRadius=18, drivingBone {boneName=spine_01}, sphereLocalOffset=<0 0 0>}
     {limitRadius=18, drivingBone {boneName=spine_02}, sphereLocalOffset=<0 0 0>}
     {limitRadius=18, drivingBone {boneName=pelvis}, sphereLocalOffset=<0 0 0>}
    ]
   }
   n31 = Anim Dynamics - Bone: dyn_coat_frdiag_02_r(ComponentPose=@n30, Alpha=1, ExternalForce="0,0,0") {
    angularDampingOverride=0.98
    bGravityOverrideInSimSpace=true
    bOverrideAngularDamping=true
    bUsePlanarLimit=false
    bUseSphericalLimits=true
    numSolverIterationsPostUpdate=4
    numSolverIterationsPreUpdate=16
    alphaBoolBlend {blendOption=Linear}
    boundBone {boneName=dyn_coat_frdiag_02_r}
    componentAppliedLinearAccClamp=<3000 3000 3000>
    componentLinearAccScale=<0.5 0.5 0.5>
    componentLinearVelScale=<0.5 0.5 0.5>
    physicsBodyDefinitions [{
      sphereCollisionRadius=0
      boundBone {boneName=dyn_coat_frdiag_02_r}
      boxExtents=<15 30 40>
      constraintSetup {twistAxis=AxisZ, angularLimitsMax=<89 30 0>, angularLimitsMin=<-89 -150 0>, angularTarget=<0 0 0>}
      localJointOffset=<0 0 10>
     }]
    sphericalLimits [
     {limitRadius=18, drivingBone {boneName=spine_01}, sphereLocalOffset=<0 0 0>}
     {limitRadius=18, drivingBone {boneName=spine_02}, sphereLocalOffset=<0 0 0>}
     {limitRadius=18, drivingBone {boneName=pelvis}, sphereLocalOffset=<0 0 0>}
    ]
   }
   n32 = Anim Dynamics - Bone: dyn_neck_tie_l(ComponentPose=@n31, Alpha=0.55, ExternalForce="0,0,0") {
    angularSpringConstant=100
    bGravityOverrideInSimSpace=true
    bUsePlanarLimit=false
    bUseSphericalLimits=true
    gravityScale=0.4
    alphaBoolBlend {blendOption=Linear}
    boundBone {boneName=dyn_neck_tie_l}
    componentAppliedLinearAccClamp=<1000 1000 1000>
    physicsBodyDefinitions [{
      collisionType=CustomSphere
      sphereCollisionRadius=1
      boundBone {boneName=dyn_neck_tie_l}
      boxExtents=<10 3 10>
      constraintSetup {angularLimitsMax=<-20 20 5>, angularLimitsMin=<-35 -20 -5>, angularTarget=<0 0 0>}
      localJointOffset=<-0.1 0 -8>
     }]
    planarLimits [{drivingBone {boneName=spine_05}, planeTransform {location=<0 -2 0>, rotation=<rot 90 40 0>}}]
    relativeSpaceBone {boneName=dyn_neck_tie_base}
    sphericalLimits [{limitRadius=80, drivingBone {boneName=spine_05}, sphereLocalOffset=<0 -73.25 3>}]
   }
   n33 = Anim Dynamics - Bone: dyn_neck_tie_r(ComponentPose=@n32, Alpha=0.45, ExternalForce="0,0,0") {
    bGravityOverrideInSimSpace=true
    gravityScale=0.35
    alphaBoolBlend {blendOption=Linear}
    boundBone {boneName=dyn_neck_tie_r}
    componentAppliedLinearAccClamp=<1000 1000 1000>
    physicsBodyDefinitions [{
      sphereCollisionRadius=0
      boundBone {boneName=dyn_neck_tie_r}
      boxExtents=<10 10 10>
      constraintSetup {angularLimitsMax=<-1 20 5>, angularLimitsMin=<-30 -20 -5>, angularTarget=<0 0 0>}
      localJointOffset=<0 0 -5>
     }]
   }
   n34 = Anim Dynamics - Bone: dyn_mug(ComponentPose=@n33, Alpha=1, ExternalForce="0,0,0") {
    bGravityOverrideInSimSpace=true
    numSolverIterationsPostUpdate=4
    numSolverIterationsPreUpdate=16
    alphaBoolBlend {blendOption=Linear}
    boundBone {boneName=dyn_mug}
    physicsBodyDefinitions [{
      sphereCollisionRadius=0
      boundBone {boneName=dyn_mug}
      boxExtents=<50 50 50>
      constraintSetup {twistAxis=AxisZ, angularLimitsMax=<10 5 0>, angularLimitsMin=<-10 0 -10>, angularTarget=<1 0 0>}
      localJointOffset=<-1 0 -20>
     }]
   }
   Output Pose(Result=Component To Local(ComponentPose=@n34)) {layerGroup=None}
  ]}
 EventGraph {nodes [
   TryGetPawnOwner()
   Event BlueprintUpdateAnimation()
  ]}
]
variables [Shorts_Alpha {type=number, default=0.6}]
