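# Animation Blueprint for the male medium Octoberfest face accessory, adding secondary physics jiggle to its beard, rose, leaf and hair bones.
# Copies pose and curves from the attached parent mesh, then runs serial single-bone Anim Dynamics nodes in component space.
# Spring-driven, heavily damped bodies are mostly limited to small linear offsets and partially blended by alpha; dyn_hair_01 instead gets angular swing.
# No variables or event logic drive it.
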AnimBlueprint /BRCosmetics/Characters/Player/Male/Medium/Bodies/M_MED_Octoberfest/Meshes/M_MED_Octoberfest_FaceAcc_AnimBP
target_skeleton=&/M_MED_Octoberfest_FaceAcc_Skeleton
graphs [
 AnimGraph {nodes [
   n1 = Anim Dynamics - Bone: dyn_beard_l(
    ComponentPose=Local To Component(LocalPose=Copy Pose From Mesh() {bCopyCurves=true, bUseAttachedParent=true})
    Alpha=0.225
    ExternalForce="0,0,0"
   ) {
    angularDampingOverride=0.985
    angularSpringConstant=350
    bAngularSpring=true
    bGravityOverrideInSimSpace=true
    bLinearSpring=true
    bOverrideAngularDamping=true
    bOverrideLinearDamping=true
    gravityScale=0.2
    linearDampingOverride=0.99
    linearSpringConstant=350
    alphaBoolBlend {blendOption=Linear}
    boundBone {boneName=dyn_beard_l}
    componentAppliedLinearAccClamp=<1000 1000 1000>
    componentLinearVelScale=<100 100 100>
    physicsBodyDefinitions [{
      sphereCollisionRadius=0
      boundBone {boneName=dyn_beard_l}
      boxExtents=<60 60 60>
      constraintSetup {
       angularTargetAxis=AxisY
       angularLimitsMax=<0 0 0>
       angularLimitsMin=<0 0 0>
       angularTarget=<0 0 0>
       linearAxesMax=<3 0 0>
       linearAxesMin=<-0.5 0 0>
      }
      localJointOffset=<0 0 0>
     }]
   }
   n2 = Anim Dynamics - Bone: dyn_beard_r(ComponentPose=@n1, Alpha=0.225, ExternalForce="0,0,0") {
    angularDampingOverride=0.985
    angularSpringConstant=350
    bAngularSpring=true
    bGravityOverrideInSimSpace=true
    bLinearSpring=true
    bOverrideAngularDamping=true
    bOverrideLinearDamping=true
    gravityScale=0.2
    linearDampingOverride=0.99
    linearSpringConstant=350
    alphaBoolBlend {blendOption=Linear}
    boundBone {boneName=dyn_beard_r}
    componentAppliedLinearAccClamp=<1000 1000 1000>
    componentLinearVelScale=<100 100 100>
    physicsBodyDefinitions [{
      sphereCollisionRadius=0
      boundBone {boneName=dyn_beard_r}
      boxExtents=<60 60 60>
      constraintSetup {
       angularTargetAxis=AxisY
       angularLimitsMax=<0 0 0>
       angularLimitsMin=<0 0 0>
       angularTarget=<0 0 0>
       linearAxesMax=<3 0 0>
       linearAxesMin=<-0.5 0 0>
      }
      localJointOffset=<0 0 0>
     }]
   }
   n3 = Anim Dynamics - Bone: dyn_rose_01(ComponentPose=@n2, Alpha=0.085, ExternalForce="0,0,0") {
    angularDampingOverride=0.985
    angularSpringConstant=350
    bAngularSpring=true
    bGravityOverrideInSimSpace=true
    bLinearSpring=true
    bOverrideAngularDamping=true
    bOverrideLinearDamping=true
    gravityScale=0.35
    linearDampingOverride=0.99
    linearSpringConstant=350
    alphaBoolBlend {blendOption=Linear}
    boundBone {boneName=dyn_rose_01}
    componentAppliedLinearAccClamp=<1000 1000 1000>
    componentLinearVelScale=<1000 1000 1000>
    physicsBodyDefinitions [{
      sphereCollisionRadius=0
      boundBone {boneName=dyn_rose_01}
      boxExtents=<40 40 10>
      constraintSetup {
       angularTargetAxis=AxisZ
       angularLimitsMax=<0 0 0>
       angularLimitsMin=<0 0 0>
       angularTarget=<0 0 0>
       linearAxesMax=<1.5 1.5 0.35>
       linearAxesMin=<-0.5 -3 -0.5>
      }
      localJointOffset=<0 0 0>
     }]
   }
   n4 = Anim Dynamics - Bone: dyn_rose_02(ComponentPose=@n3, Alpha=0.075, ExternalForce="0,0,0") {
    angularDampingOverride=0.985
    angularSpringConstant=600
    bAngularSpring=true
    bGravityOverrideInSimSpace=true
    bLinearSpring=true
    bOverrideAngularDamping=true
    bOverrideLinearDamping=true
    gravityScale=0.25
    linearDampingOverride=0.99
    linearSpringConstant=700
    alphaBoolBlend {blendOption=Linear}
    boundBone {boneName=dyn_rose_02}
    componentAppliedLinearAccClamp=<1000 1000 1000>
    componentLinearVelScale=<1000 1000 1000>
    physicsBodyDefinitions [{
      sphereCollisionRadius=0
      boundBone {boneName=dyn_rose_02}
      boxExtents=<40 40 15>
      constraintSetup {
       angularTargetAxis=AxisY
       twistAxis=AxisY
       angularLimitsMax=<0 0 0>
       angularLimitsMin=<0 0 0>
       angularTarget=<0 0 0>
       linearAxesMax=<2 3.5 0.15>
       linearAxesMin=<-0.025 -1 -0.25>
      }
      localJointOffset=<0 0 0>
     }]
   }
   n5 = Anim Dynamics - Bone: dyn_rose_03(ComponentPose=@n4, Alpha=0.065, ExternalForce="0,0,0") {
    angularDampingOverride=0.985
    angularSpringConstant=600
    bAngularSpring=true
    bGravityOverrideInSimSpace=true
    bLinearSpring=true
    bOverrideAngularDamping=true
    bOverrideLinearDamping=true
    gravityScale=0.275
    linearDampingOverride=0.99
    linearSpringConstant=600
    alphaBoolBlend {blendOption=Linear}
    boundBone {boneName=dyn_rose_03}
    componentAppliedLinearAccClamp=<1000 1000 1000>
    componentLinearVelScale=<1000 1000 1000>
    physicsBodyDefinitions [{
      sphereCollisionRadius=0
      boundBone {boneName=dyn_rose_03}
      boxExtents=<20 20 5>
      constraintSetup {
       angularTargetAxis=AxisY
       angularLimitsMax=<0 0 0>
       angularLimitsMin=<0 0 0>
       angularTarget=<0 0 0>
       linearAxesMax=<1.25 2.5 0.45>
       linearAxesMin=<-0.5 -2 -0.25>
      }
      localJointOffset=<0 0 0>
     }]
   }
   n6 = Anim Dynamics - Bone: dyn_leaf_01(ComponentPose=@n5, Alpha=0.45, ExternalForce="0,0,0") {
    angularDampingOverride=0.985
    angularSpringConstant=500
    bAngularSpring=true
    bGravityOverrideInSimSpace=true
    bLinearSpring=true
    bOverrideAngularDamping=true
    bOverrideLinearDamping=true
    gravityScale=0.085
    linearDampingOverride=0.99
    linearSpringConstant=600
    alphaBoolBlend {blendOption=Linear}
    boundBone {boneName=dyn_leaf_01}
    componentAppliedLinearAccClamp=<1000 1000 1000>
    componentLinearVelScale=<100 100 100>
    physicsBodyDefinitions [{
      sphereCollisionRadius=0
      boundBone {boneName=dyn_leaf_01}
      boxExtents=<30 30 30>
      constraintSetup {
       angularTargetAxis=AxisY
       angularLimitsMax=<0 0 0>
       angularLimitsMin=<0 0 0>
       angularTarget=<0 0 0>
       linearAxesMax=<3 2 0.5>
       linearAxesMin=<-0.5 0 -0.5>
      }
      localJointOffset=<0 0 0>
     }]
   }
   n7 = Anim Dynamics - Bone: dyn_leaf_02(ComponentPose=@n6, Alpha=0.3, ExternalForce="0,0,0") {
    angularDampingOverride=0.985
    angularSpringConstant=300
    bAngularSpring=true
    bGravityOverrideInSimSpace=true
    bLinearSpring=true
    bOverrideAngularDamping=true
    bOverrideLinearDamping=true
    gravityScale=0.15
    linearDampingOverride=0.99
    linearSpringConstant=200
    alphaBoolBlend {blendOption=Linear}
    boundBone {boneName=dyn_leaf_02}
    componentAppliedLinearAccClamp=<1000 1000 1000>
    componentLinearVelScale=<100 100 100>
    physicsBodyDefinitions [{
      sphereCollisionRadius=0
      boundBone {boneName=dyn_leaf_02}
      boxExtents=<30 45 7>
      constraintSetup {
       angularTargetAxis=AxisY
       angularLimitsMax=<0 0 0>
       angularLimitsMin=<0 0 0>
       angularTarget=<0 0 0>
       linearAxesMax=<3 2 0.5>
       linearAxesMin=<-0.5 0 -0.5>
      }
      localJointOffset=<0 0 0>
     }]
   }
   n8 = Anim Dynamics - Bone: dyn_leaf_03(ComponentPose=@n7, Alpha=0.35, ExternalForce="0,0,0") {
    angularDampingOverride=0.985
    angularSpringConstant=500
    bAngularSpring=true
    bGravityOverrideInSimSpace=true
    bLinearSpring=true
    bOverrideAngularDamping=true
    bOverrideLinearDamping=true
    gravityScale=0.3
    linearDampingOverride=0.99
    linearSpringConstant=500
    alphaBoolBlend {blendOption=Linear}
    boundBone {boneName=dyn_leaf_03}
    componentAppliedLinearAccClamp=<1000 1000 1000>
    componentLinearVelScale=<100 100 100>
    physicsBodyDefinitions [{
      sphereCollisionRadius=0
      boundBone {boneName=dyn_leaf_03}
      boxExtents=<20 10 15>
      constraintSetup {
       angularTargetAxis=AxisY
       angularLimitsMax=<0 0 0>
       angularLimitsMin=<0 0 0>
       angularTarget=<0 0 0>
       linearAxesMax=<3 2 0.5>
       linearAxesMin=<-0.5 0 -0.5>
      }
      localJointOffset=<0 0 0>
     }]
   }
   n9 = Anim Dynamics - Bone: dyn_leaf_04(ComponentPose=@n8, Alpha=0.35, ExternalForce="0,0,0") {
    angularDampingOverride=0.985
    angularSpringConstant=500
    bAngularSpring=true
    bGravityOverrideInSimSpace=true
    bLinearSpring=true
    bOverrideAngularDamping=true
    bOverrideLinearDamping=true
    gravityScale=0.2
    linearDampingOverride=0.99
    linearSpringConstant=750
    alphaBoolBlend {blendOption=Linear}
    boundBone {boneName=dyn_leaf_04}
    componentAppliedLinearAccClamp=<1000 1000 1000>
    componentLinearVelScale=<100 100 100>
    physicsBodyDefinitions [{
      sphereCollisionRadius=0
      boundBone {boneName=dyn_leaf_04}
      boxExtents=<65 23 35>
      constraintSetup {
       angularTargetAxis=AxisY
       angularLimitsMax=<0 0 0>
       angularLimitsMin=<0 0 0>
       angularTarget=<0 0 0>
       linearAxesMax=<3 2 0.5>
       linearAxesMin=<-0.5 0 -0.5>
      }
      localJointOffset=<0 0 0>
     }]
   }
   n10 = Anim Dynamics - Bone: dyn_leaf_05(ComponentPose=@n9, Alpha=0.425, ExternalForce="0,0,0") {
    angularDampingOverride=0.985
    angularSpringConstant=500
    bAngularSpring=true
    bGravityOverrideInSimSpace=true
    bLinearSpring=true
    bOverrideAngularDamping=true
    bOverrideLinearDamping=true
    gravityScale=0.275
    linearDampingOverride=0.99
    linearSpringConstant=325
    alphaBoolBlend {blendOption=Linear}
    boundBone {boneName=dyn_leaf_05}
    componentAppliedLinearAccClamp=<1000 1000 1000>
    componentLinearVelScale=<100 100 100>
    physicsBodyDefinitions [{
      sphereCollisionRadius=0
      boundBone {boneName=dyn_leaf_05}
      boxExtents=<15 35 33>
      constraintSetup {
       angularTargetAxis=AxisY
       angularLimitsMax=<0 0 0>
       angularLimitsMin=<0 0 0>
       angularTarget=<0 0 0>
       linearAxesMax=<3 2 0.5>
       linearAxesMin=<-0.5 0 -0.5>
      }
      localJointOffset=<0 0 0>
     }]
   }
   n11 = Anim Dynamics - Bone: dyn_hair_01(ComponentPose=@n10, Alpha=0.75) {
    angularDampingOverride=0.985
    angularSpringConstant=350
    bAngularSpring=true
    bGravityOverrideInSimSpace=true
    bLinearSpring=true
    bOverrideAngularDamping=true
    bOverrideLinearDamping=true
    gravityScale=0.7
    linearDampingOverride=0.99
    linearSpringConstant=350
    alphaBoolBlend {blendOption=Linear}
    boundBone {boneName=dyn_hair_01}
    componentAppliedLinearAccClamp=<1000 1000 1000>
    componentLinearVelScale=<100 100 100>
    physicsBodyDefinitions [{
      sphereCollisionRadius=0
      boundBone {boneName=dyn_hair_01}
      boxExtents=<30 30 30>
      constraintSetup {angularTargetAxis=AxisY, twistAxis=AxisZ, angularLimitsMax=<25 0 0>, angularLimitsMin=<-25 0 0>, angularTarget=<0 0 0>}
      localJointOffset=<0 -1 -15>
     }]
   }
   n12 = Anim Dynamics - Bone: dyn_hair_02(ComponentPose=@n11, Alpha=0.25) {
    angularDampingOverride=0.985
    angularSpringConstant=500
    bAngularSpring=true
    bGravityOverrideInSimSpace=true
    bLinearSpring=true
    bOverrideAngularDamping=true
    bOverrideLinearDamping=true
    gravityScale=0.2
    linearDampingOverride=0.99
    linearSpringConstant=325
    alphaBoolBlend {blendOption=Linear}
    boundBone {boneName=dyn_hair_02}
    componentAppliedLinearAccClamp=<1000 1000 1000>
    componentLinearVelScale=<100 100 100>
    physicsBodyDefinitions [{
      sphereCollisionRadius=0
      boundBone {boneName=dyn_hair_02}
      boxExtents=<15 35 33>
      constraintSetup {
       angularTargetAxis=AxisY
       angularLimitsMax=<0 0 0>
       angularLimitsMin=<0 0 0>
       angularTarget=<0 0 0>
       linearAxesMax=<3 2 0.5>
       linearAxesMin=<-0.5 0 -0.5>
      }
      localJointOffset=<0 0 0>
     }]
   }
   Output Pose(Result=Component To Local(ComponentPose=@n12)) {layerGroup=None}
   AnimGraphNode_SubInput_0 = Input Pose()
  ]}
 EventGraph {nodes [Event BlueprintUpdateAnimation()]}
]
variables []
